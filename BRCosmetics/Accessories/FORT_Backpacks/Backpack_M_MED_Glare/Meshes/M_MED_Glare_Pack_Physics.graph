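# Physics asset for the M_MED_Glare backpack cosmetic, simulating its dangling straps, closure and four wheels.
# The pack body is a capsule plus box that anchors box bodies for each dynamic bone.
# Straps A and B are tightly limited three-link chains, while strap C limits its root and frees swing on its lower links.
# Wheels lock swing and rotate only on twist with differing limits, and the closure locks twist but allows limited swing.
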PhysicsAsset /BRCosmetics/Accessories/FORT_Backpacks/Backpack_M_MED_Glare/Meshes/M_MED_Glare_Pack_Physics
bodies {
 dyn_backpack_straps_A [{shapeType=Box, extentX=2.419, extentY=1.7311, extentZ=0.7327, shapeName=dyn_backpack_straps_A_box, rotation=<rot 0 -90 0>}]
 dyn_backpack_straps_A1 [{
   shapeType=Box
   extentX=1.9334
   extentY=0.7694
   extentZ=2.4836
   shapeName=dyn_backpack_straps_A1_box
   center=<0.0235 -0.0671 0.0215>
   rotation=<rot -0.006 1.0656 -90.7151>
  }]
 dyn_backpack_straps_A2 [{
   shapeType=Box
   extentX=2.0811
   extentY=0.851
   extentZ=2.5017
   shapeName=dyn_backpack_straps_A2_box
   center=<0.3517 0.0013 -0.0054>
   rotation=<rot -1.2812 -4.7579 -92.9942>
  }]
 dyn_backpack_straps_B [{
   shapeType=Box
   extentX=2.4738
   extentY=0.6887
   extentZ=1.4285
   shapeName=dyn_backpack_straps_A_box
   center=<-0.0527 -0.1175 0.0559>
   rotation=<rot 0.0772 -89.8489 -89.3438>
  }]
 dyn_backpack_straps_B1 [{shapeType=Box, extentX=1.0369, extentY=2.4179, extentZ=0.7427, shapeName=dyn_backpack_straps_B1_box, center=<-0.2306 -0.0922 0.086>}]
 dyn_backpack_straps_B2 [{shapeType=Box, extentX=1.7634, extentY=2.2813, extentZ=0.7427, shapeName=dyn_backpack_straps_B2_box, center=<0.4769 -0.0845 0.0844>}]
 dyn_backpack_straps_C_01 [{
   shapeType=Box
   extentX=1.4655
   extentY=0.5
   extentZ=1.848
   shapeName=dyn_backpack_straps_C_01_box
   center=<3.2257 0.0216 0.2242>
   rotation=<rot -3.3396 90.8854 -89.1966>
  }]
 dyn_backpack_straps_C_02 [{
   shapeType=Box
   extentX=1.51
   extentY=1.01
   extentZ=1.01
   shapeName=dyn_backpack_straps_C_02_box
   center=<1.9781 -0.0133 0.0613>
   rotation=<rot 0 -88.4195 -89.2779>
  }]
 dyn_backpack_straps_C_03 [{
   shapeType=Box
   extentX=2.8785
   extentY=2.6693
   extentZ=2.99
   shapeName=dyn_backpack_straps_C_03_box
   center=<2.9793 0.8702 0.143>
   rotation=<rot 0 105.6248 -86.9342>
  }]
 dyn_closure [{
   shapeType=Box
   extentX=1.6601
   extentY=1.0615
   extentZ=0.853
   shapeName=dyn_closure_box
   center=<2.6571 0.0267 0.0152>
   rotation=<rot 0.0091 89.2179 -89.515>
  }]
 dyn_l_wheels_down [{shapeType=Box, extentX=5.2871, extentY=6, extentZ=6, shapeName=dyn_r_wheels_down_box, center=<-3.1577 0.4984 0>}]
 dyn_l_wheels_up [{shapeType=Box, extentX=5.2871, extentY=6, extentZ=6, shapeName=dyn_r_wheels_down_box, center=<-3.1577 0.4984 0>}]
 dyn_r_wheels_down [{shapeType=Box, extentX=5.2871, extentY=6, extentZ=6, shapeName=dyn_r_wheels_down_box, center=<3.0719 0.4984 0>}]
 dyn_r_wheels_up [{shapeType=Box, extentX=5.2871, extentY=6, extentZ=6, shapeName=dyn_r_wheels_down_box, center=<3.0719 0.4984 0>}]
 pack [
  {shapeType=Capsule, length=7, radius=6, shapeName=pack_capsule, center=<-4.7343 -4.7961 -7.1876>, rotation=<rot 44.1873 -0.028 6.862>}
  {
   shapeType=Box
   extentX=4.3797
   extentY=15.1855
   extentZ=25.0793
   shapeName=pack_box
   center=<-1.098 0.3279 -6.7281>
   rotation=<rot -3.0085 -90 45.7952>
  }
 ]
}
constraints [
 {bone1Name=dyn_closure, bone2Name=pack, swing1LimitDegrees=45, swing2LimitDegrees=8, twistLimitDegrees=45, twistMotion=Locked}
 {
  bone1Name=dyn_l_wheels_down
  bone2Name=pack
  swing1LimitDegrees=45
  swing1Motion=Locked
  swing2LimitDegrees=45
  swing2Motion=Locked
  twistLimitDegrees=100
 }
 {
  bone1Name=dyn_r_wheels_down
  bone2Name=pack
  swing1LimitDegrees=45
  swing1Motion=Locked
  swing2LimitDegrees=45
  swing2Motion=Locked
  twistLimitDegrees=110
 }
 {
  bone1Name=dyn_l_wheels_up
  bone2Name=pack
  swing1LimitDegrees=45
  swing1Motion=Locked
  swing2LimitDegrees=45
  swing2Motion=Locked
  twistLimitDegrees=70
 }
 {
  bone1Name=dyn_r_wheels_up
  bone2Name=pack
  swing1LimitDegrees=45
  swing1Motion=Locked
  swing2LimitDegrees=45
  swing2Motion=Locked
  twistLimitDegrees=90
 }
 {bone1Name=dyn_backpack_straps_C_01, bone2Name=pack, swing1LimitDegrees=50, swing2LimitDegrees=50, twistLimitDegrees=45}
 {
  bone1Name=dyn_backpack_straps_C_02
  bone2Name=dyn_backpack_straps_C_01
  swing1LimitDegrees=45
  swing1Motion=Free
  swing2LimitDegrees=45
  swing2Motion=Free
  twistLimitDegrees=10
 }
 {
  bone1Name=dyn_backpack_straps_C_03
  bone2Name=dyn_backpack_straps_C_02
  swing1LimitDegrees=45
  swing1Motion=Free
  swing2LimitDegrees=45
  swing2Motion=Free
  twistLimitDegrees=10
 }
 {bone1Name=dyn_backpack_straps_B1, bone2Name=dyn_backpack_straps_B, swing1LimitDegrees=8, swing2LimitDegrees=13, twistLimitDegrees=4}
 {bone1Name=dyn_backpack_straps_B2, bone2Name=dyn_backpack_straps_B1, swing1LimitDegrees=8, swing2LimitDegrees=10, twistLimitDegrees=4}
 {bone1Name=dyn_backpack_straps_A1, bone2Name=dyn_backpack_straps_A, swing1LimitDegrees=8, swing2LimitDegrees=13, twistLimitDegrees=4}
 {bone1Name=dyn_backpack_straps_A2, bone2Name=dyn_backpack_straps_A1, swing1LimitDegrees=8, swing2LimitDegrees=10, twistLimitDegrees=4}
 {bone1Name=dyn_backpack_straps_A, bone2Name=pack, swing1LimitDegrees=8, swing2LimitDegrees=19, twistLimitDegrees=4}
 {bone1Name=dyn_backpack_straps_B, bone2Name=pack, swing1LimitDegrees=8, swing2LimitDegrees=19, twistLimitDegrees=4}
]
